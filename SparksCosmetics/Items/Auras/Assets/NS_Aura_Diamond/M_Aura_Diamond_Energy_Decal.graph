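# Translucent unlit Niagara material for the Diamond aura cosmetic, drawing a decal-like energy glow on scene geometry behind the particle.
# Panned caustics drive brightness and hue-shift the particle color, plus a Fresnel from flattened scene WorldNormal, constant on mobile.
# Opacity combines that Fresnel, a UV sphere mask and inverted depth fade limiting visibility to nearby surfaces.
# Two counter-panned crystal-facet samples modulate opacity, with Niagara dynamic parameters controlling exponent, offset and glow.
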Material /SparksCosmetics/Items/Auras/Assets/NS_Aura_Diamond/M_Aura_Diamond_Energy_Decal
graph {
 properties {
  bUsedWithNiagaraMeshParticles=true
  bUsedWithNiagaraSprites=true
  blendMode=BLEND_Translucent
  pixelDepthOffsetMode=PDOM_Legacy
  shadingModel=MSM_Unlit
  usedShadingModels=Unlit
 }
 nodes [
  n1 = Multiply(
   A=TextureCoordinate()
   B=AppendVector(
    A=ScalarParameter() {defaultValue=1.5, group=Caustics, parameterName=AlphaScaleX}
    B=ScalarParameter() {defaultValue=0.5, group=Caustics, parameterName=AlphaScaleY}
   )
  )
  n2 = ParticleRandom()
  n3 = ScalarParameter() {group=Caustics, parameterName=AlphaSpeedX}
  n4 = ScalarParameter() {defaultValue=0.75, group=Caustics, parameterName=AlphaSpeedY}
  n5 = TextureCoordinate()
  n6 = TextureSample!(
   UVs=Panner(Coordinate=Add(A=@n1, B=@n2), Speed=AppendVector(A=@n3, B=@n4)) {speedY=0.75}
   Tex
   "DDX(UVs)"=DDX(Value=@n5)
   "DDY(UVs)"=DDY(Value=@n5)
   Apply View MipBias
  ) {
   mipValueMode=TMVM_Derivative
   samplerType=SAMPLERTYPE_LinearGrayscale
   texture=/Game/Global/Textures/Default/Caustics/T_Nosie_Caustics_Grayscale
  }
  n7 = Add(A=Multiply(A=@n6.R, B=ScalarParameter() {defaultValue=0.5, group=Caustics, parameterName=Caustics}) {constB=10}) {constB=0.25}
  n8 = MaterialFunctionCall(
   Hue Shift Percentage=Multiply(A=Saturate(None=@n7), B=ScalarParameter() {defaultValue=0.25, group=Caustics, parameterName=HueShift})
   Texture=ParticleColor().RGB
  ) {materialFunction=&HueShift}
  n9 = SceneTexture() {sceneTextureId=PPI_WorldNormal}
  n10 = Fresnel(
   ExponentIn=DynamicParameter() {defaultValue=<rgba 4 1 1 0.08>, paramNames [FresnelExponent, U Offset, V Offset, Glow Amount]}.FresnelExponent
   Normal=MaterialFunctionCall(
    Normal=ComponentMask(None=@n9.Color) {b=true, g=true, r=true}
    Flatness=ScalarParameter() {defaultValue=0.15, group=Fresnel, parameterName=FlattenNormal}
   ) {materialFunction=&FlattenNormal}.Result
  ) {baseReflectFraction=0, exponent=1}
  n11 = ShadingPathSwitch(Default=@n10, Deferred=@n10, Forward=VectorParameter() {group=Fresnel, parameterName=Forward, defaultValue=<rgba 0 0 1 1>}.RGB, Mobile=@n10)
  n12 = ShadingPathSwitch(Default=@n11, Deferred=@n11, Mobile=Constant() {r=0.05}) {desc=Migrated from FeatureLevelSwitch}
  MaterialFunctionCall(Emissive=Add(A=Multiply(A=@n7, B=@n8.Result), B=Multiply(A=@n8.Result, B=@n12))) {materialFunction=&MF_FX_EmissiveAdjustments}
  n13 = ParticleColor()
  n14 = TextureCoordinate()
  n15 = Constant() {r=0.5}
  n16 = ScalarParameter() {defaultValue=0.5, group=Opacity, parameterName=SphereMaskRadius}
  n17 = ScalarParameter() {group=Opacity, parameterName=SphereMaskHardness}
  n18 = DepthFade(FadeDistance=ScalarParameter() {defaultValue=20, group=Opacity, parameterName=VisibleDepthDistance}) {fadeDistanceDefault=80}
  n19 = Multiply(
   A=Multiply(A=@n12, B=SphereMask(A=@n14, B=@n15, Radius=@n16, Hardness=@n17) {hardnessPercent=0})
   B=DepthFade(Opacity=OneMinus(None=@n18), FadeDistance=ScalarParameter() {defaultValue=15, group=Opacity, parameterName=DepthFade}) {fadeDistanceDefault=15}
  )
  n20 = ScalarParameter() {defaultValue=1, group=Opacity, parameterName=Use Facets}
  n21 = Constant2Vector() {g=1, r=-0.2}
  n22 = ScalarParameter() {defaultValue=1, group=Facets, parameterName=Facets Pan Speed Mult}
  n23 = DynamicParameter() {defaultValue=<rgba 4 1 1 0.08>, paramNames [FresnelExponent, U Offset, V Offset, Glow Amount]}
  n24 = Add(
   A=Multiply(A=TextureCoordinate(), B=ScalarParameter() {defaultValue=1, group=Facets, parameterName=Facets Scale Mult})
   B=AppendVector(A=@n23.U Offset, B=@n23.U Offset)
  )
  n25 = TextureSampleParameter2D(UVs=Panner(Coordinate=@n24, Speed=Multiply(A=@n21, B=@n22)) {speedX=-0.2, speedY=1}) {group=Facets, parameterName=Facets Texture, texture=/Game/Accessories/FORT_Backpacks/Backpack_F_MED_IchorIncisor/FX/_old/T_CrystalShield}
  n26 = Constant2Vector() {g=0.6, r=0.4}
  n27 = TextureSampleParameter2D(UVs=Panner(Coordinate=@n24, Speed=Multiply(A=@n22, B=@n26)) {speedX=0.4, speedY=0.6}) {group=Facets, parameterName=Facets Texture, texture=&T_CrystalShield}
  n28 = Multiply(A=@n25.R, B=@n27.G)
  n29 = MaterialFunctionCall(Input=Saturate(None=Multiply(A=@n28) {constB=2}), Input Low=Constant() {r=0.2}, Input High=Constant() {r=0.6}) {materialFunction=&RemapValue_0-1}
  Multiply(
   A=Multiply(A=Multiply(A=@n13.A, B=@n19), B=ScalarParameter() {defaultValue=1, group=Opacity, parameterName=Opacity})
   B=Add(
    A=DynamicParameter() {defaultValue=<rgba 4 1 1 0.08>, paramNames [FresnelExponent, U Offset, V Offset, Glow Amount]}.Glow Amount
    B=Multiply(A=@n20, B=@n29.Result)
   )
  )
 ]
}
