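# Additive unlit Niagara sprite material that draws animated sparkles for the Sparks Diamond aura cosmetic.
# Three panning random-dot layers, offset per particle, are summed and reduced toward the sprite edges by a sphere mask.
# Color scrolls through the Rainbow aura's curve atlas, shifted by dynamic parameters and blended with an End Color along V.
# The dots and particle color feed HDR-derived emissive, while particle alpha gets depth fade on deferred shading, not forward or mobile.
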Material /SparksCosmetics/Items/Auras/Assets/NS_Aura_Diamond/M_Aura_Sparkles
graph {
 properties {
  bAutomaticallySetUsageInEditor=false
  bUsedWithNiagaraSprites=true
  blendMode=BLEND_Additive
  shadingModel=MSM_Unlit
  usedShadingModels=Unlit
 }
 nodes [
  n1 = DynamicParameter() {defaultValue=<rgba 1.2781 1 1 1>, paramNames [Color Curve, Color Curve Scale, --, --]}
  n2 = MaterialFunctionCall() {materialFunction=&LinearGradient}
  n3 = Time()
  n4 = ScalarParameter() {defaultValue=0.5, parameterName=Color Pan Speed Mult}
  n5 = Add(
   A=Multiply(A=@n1.Color Curve, B=Divide(B=@n1.Color Curve Scale))
   B=Add(A=Multiply(A=@n1.Color Curve Scale, B=@n2.UGradient), B=Multiply(A=@n3, B=@n4) {constB=0.1})
  )
  n6 = CurveAtlasRowParameter(CurveTime=Subtract(A=@n5, B=Floor(None=@n5))) {
   atlas=/SparksCosmetics/Items/Auras/Assets/NS_Aura_Rainbow/CA_Aura_Rainbow
   curve=/SparksCosmetics/Items/Auras/Assets/NS_Aura_Rainbow/CC_Rainbow
   parameterName=Color Curve
   sortPriority=2
  }
  n7 = VectorParameter() {parameterName=End Color, defaultValue=<rgba 0.1146 0.3402 1 0.75>}
  n8 = LinearInterpolate(A=@n6, B=@n7.RGB, Alpha=@n7.RGB)
  n9 = ScalarParameter() {defaultValue=0.25, parameterName=End Coverage}
  n10 = MaterialFunctionCall(Input=@n2.VGradient, Input Low=@n9, Input High=Add(A=@n9, B=ScalarParameter() {defaultValue=0.5, parameterName=End Transition Width})) {materialFunction=&RemapValue_0-1}
  n11 = ParticleRandom()
  n12 = Sine(None=@n11)
  n13 = Multiply(
   A=AppendVector(A=ScalarParameter() {defaultValue=1, parameterName=X Scale Mult}, B=ScalarParameter() {defaultValue=1, parameterName=Y Scale Mult})
   B=Add(A=TextureCoordinate(), B=AppendVector(A=@n11, B=@n12))
  )
  n14 = VectorParameter() {parameterName=Detail Texture, sortPriority=4, channelNames {a=V Speed, b=U Speed, g=V Scale, r=U Scale}, defaultValue=<rgba 3 4 0.3 2>}
  n15 = AppendVector(A=@n14.U Scale, B=@n14.U Scale)
  n16 = Multiply(A=Time(), B=ScalarParameter() {defaultValue=1, parameterName=Pan Speed Mult, sortPriority=3})
  n17 = TextureSample(UVs=Panner(Coordinate=Multiply(A=@n13, B=@n15), Time=@n16, Speed=AppendVector(A=@n14.U Speed, B=@n14.U Speed)) {speedX=0.3, speedY=-4}) {samplerType=SAMPLERTYPE_LinearColor, texture=/Game/Effects/Fort_Effects/Textures/Tile/T_Random_Dots}
  n18 = VectorParameter() {parameterName=Base Texture, sortPriority=5, channelNames {a=V Speed, b=U Speed, g=V Scale, r=U Scale}, defaultValue=<rgba 1.5 2 -0.3 1.5>}
  n19 = AppendVector(A=@n18.U Scale, B=@n18.U Scale)
  n20 = TextureSample(UVs=Panner(Coordinate=Multiply(A=@n13, B=@n19), Time=@n16, Speed=AppendVector(A=@n18.U Speed, B=@n18.U Speed)) {speedX=-0.3, speedY=-3}) {samplerType=SAMPLERTYPE_LinearColor, texture=&T_Random_Dots}
  n21 = VectorParameter() {parameterName=Macro Texture, sortPriority=7, channelNames {a=V Speed, b=U Speed, g=V Scale, r=U Scale}, defaultValue=<rgba 1 1.3333 0.5 0.5>}
  n22 = Panner(Coordinate=Multiply(A=@n13, B=AppendVector(A=@n21.U Scale, B=@n21.U Scale)), Time=@n16, Speed=AppendVector(A=@n21.U Speed, B=@n21.U Speed)) {speedX=0.5, speedY=-1}
  n23 = SphereMask(A=TextureCoordinate(), B=Constant() {r=0.5}) {attenuationRadius=1, hardnessPercent=0}
  n24 = Subtract(
   A=Add(
    A=Add(A=@n17.R, B=@n20.R)
    B=TextureSample(UVs=@n22) {samplerType=SAMPLERTYPE_LinearColor, texture=&T_Random_Dots}.R
   )
   B=Multiply(A=OneMinus(None=@n23)) {constB=2}
  )
  n25 = Multiply(A=Saturate(None=LinearInterpolate(A=@n8, B=@n6, Alpha=@n10.Result)), B=Saturate(None=@n24))
  n26 = ParticleColor()
  MaterialFunctionCall(Emissive=MaterialFunctionCall(LDR Input=Multiply(A=@n25, B=@n26.RGB), HDR Tint=@n24) {materialFunction=&DeriveHDRfromLDR}.RemappedRange) {materialFunction=&MF_FX_EmissiveAdjustments}
  Multiply(A=@n11) {constB=0.5}
  n27 = DepthFade(Opacity=@n26.A, FadeDistance=ScalarParameter() {defaultValue=5, parameterName=Depth Fade Distance})
  ShadingPathSwitch(Default=@n27, Deferred=@n27, Forward=@n26.A, Mobile=@n26.A)
 ]
}
