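# Diamond aura cosmetic for Sparks: a Niagara system that surrounds a character's skeletal mesh with glows, smoke, stars, decal sprites and drifting diamond shards.
# Most emitters sample indexed bones, several interpolating toward a companion bone from system arrays to cover limb segments.
# Flame ribbons, one per spawn group, inherit bone velocity and fade through high-speed control.
# IsFrontend and IsLargeBody adjust brightness and size; two Minimal emitters are disabled.
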NiagaraSystem /SparksCosmetics/Items/Auras/Assets/NS_Aura_Diamond/NS_Aura_Diamond
dependencies {
 usedDataInterfaces [
  &NiagaraDataInterfaceArrayFloat
  &NiagaraDataInterfaceColorCurve
  &NiagaraDataInterfaceCurve
  &NiagaraDataInterfaceArrayInt32
  &NiagaraDataInterfaceVector2DCurve
  &NiagaraDataInterfaceVectorCurve
  &NiagaraDataInterfaceSkeletalMesh
  &NiagaraDataInterfaceArrayFloat3
 ]
 usedDynamicInputs [
  &MakeVector
  &MakeCustomVector2DFromBool
  &SimulationPosition
 ]
 usedModules [
  &self:SystemSpawnScript.SystemScriptSource.SystemScriptGraph.NiagaraNodeAssignment_0.SetVariables_#1
  &self:SystemSpawnScript.SystemScriptSource.SystemScriptGraph.NiagaraNodeAssignment_3.SetVariables_#2
  &SystemState
 ]
}
emitters {
 Bits {
  emitterUpdateScript [
   &EmitterState()
   &SpawnBurst_Instantaneous(Spawn Count=14, Use Spawn Probability=false, UseLoopCountLimit=true)
  ]
  particleSpawnScript [
   /SetVariables(
    Particles.Partition:NiagaraInt32=&SelectIntFromArray(
     Direct Array Index=&ReturnExecIndex()
     Int Selection Array=DataInterface() {propertyValues {IntData [1,1,2,2,3,3,4,4,5,5,5,6,6,6]}}
    )
    Particles.PartitionIndex:NiagaraFloat=&SelectFloatFromArray(
     Direct Array Index=&ReturnExecIndex()
     Float Selection Array=DataInterface() {propertyValues {FloatData [0,0.5,0,0.5,0,0.5,0,0.5,0,0.3333,0.6667,0,0.3333,0.6667]}}
    )
    Particles.SymmetricalScalar:NiagaraFloat=&SelectFloatFromArray(
     Direct Array Index=&ReturnExecIndex()
     Float Selection Array=DataInterface() {propertyValues {FloatData [0.2,0,0.2,0,0.6,0.4,0.6,0.4,1,0.8667,0.7333,1,0.8667,0.7333]}}
    )
   ) {isSetParameters=true}
   &InitializeParticle(
    Color=&MakeCustomLinearColorFromBool(Bool=$User.IsFrontend, True Color=<rgba 1 1 1 0.4>, False Color=<rgba 1 1 1 1>)
    Sprite Size Mode=::NewEnumerator3
    Sprite UV Mode=::NewEnumerator1
    Uniform Sprite Size=&Multiply_Float(
     A=&Lerp_Float(A=25, B=10, Alpha=$Particles.SymmetricalScalar)
     B=&MakeCustomFloatFromBool(Bool=$User.IsLargeBody, True Float=1.5, False Float=1)
    )
    UsePositionOffset=false
    Write Material Random=true
   )
   &CameraOffset(Camera Offset Amount=5) {enabled=false}
   &DynamicMaterialParameters(
    Index 0 Param 1=&RandomRangeFloat(Minimum=0, Maximum=1)
    Index 0 Param 2=0.75
    Module.Index0Param0Write=true
    Module.Index0Param1Write=true
    Module.Index0Param2Write=false
    Module.Index0Param3Write=false
   )
  ]
  particleUpdateScript [
   &ParticleState(
    Kill Particles When Lifetime Has Elapsed=false
    Let Infinitely Lived Particles Die When Emitter Deactivates=true
    Loop Particles Lifetime=true
   )
   &SkeletalMeshLocation(
    Bone / Socket Index=$Particles.Partition
    Bone Sampling Mode=::NewEnumerator4
    MaskSpawnGroup=false
    Orientation Sampling=::NewEnumerator1
    Position Sampling=::NewEnumerator2
    Sampled Position Offset=<0 0 2>
    Sampled Position Offset Coordinate Space:ENiagaraSkelMeshTransforms=::NewEnumerator1
    Skeletal Mesh=$System.Skeletal Mesh
   )
   &SkeletalMeshLocation(
    Bone / Socket Index=&SelectIntFromArray(Direct Array Index=$Particles.Partition, Int Selection Array=$System.CompanionBone)
    Bone Sampling Mode=::NewEnumerator4
    "Interpolate Mesh Sampling (Update)"=true
    MaskSpawnGroup=false
    Orientation Sampling=::NewEnumerator1
    Position Sampling=::NewEnumerator2
    Sampled Position Offset=<0 0 2>
    Sampled Position Offset Coordinate Space:ENiagaraSkelMeshTransforms=::NewEnumerator1
    Skeletal Mesh=$System.Skeletal Mesh
   ) {module=SkeletalMeshLocation001}
   /SetVariables(Particles.Position=&LerpPosition(A=$Particles.SkeletalMeshLocation.SampledPosition, B=$Particles.SkeletalMeshLocation001.SampledPosition, Alpha=$Particles.PartitionIndex)) {isSetParameters=true}
   &Color(Color=&ColorFromCurve(
     ColorCurve=DataInterface() {propertyValues {
       AlphaCurve {keys [<at 0 1 RCIM_Cubic>]}
       BlueCurve {keys [<at 0 0 RCIM_Cubic>,<at 0.5 0>,<at 1 1 RCIM_Cubic>]}
       GreenCurve {keys [<at 0 0 RCIM_Cubic>,<at 0.5 1>,<at 1 0 RCIM_Cubic>]}
       RedCurve {keys [<at 0 1 RCIM_Cubic>,<at 0.5 0>,<at 1 0 RCIM_Cubic>]}
      }}
     CurveIndex=$Particles.SymmetricalScalar
    )) {enabled=false}
  ]
  rendererClasses [&NiagaraSpriteRendererProperties]
  renderers [
   {rendererClass=&NiagaraSpriteRendererProperties}
   {rendererClass=&NiagaraSpriteRendererProperties, rendererIndex=1}
  ]
 }
 Decal {
  emitterUpdateScript [
   &EmitterState()
   &SpawnRate(SpawnRate=10, Use Spawn Probability=false)
   &SpawnBurst_Instantaneous(Spawn Count=4, Use Spawn Probability=false, UseLoopCountLimit=true)
  ]
  particleSpawnScript [
   &FN_SetSpawnGroup()
   &InitializeParticle(
    Color Mode=::NewEnumerator0
    Sprite Rotation Angle=&RandomRangeFloat(Minimum=-30, Maximum=30)
    Sprite Rotation Mode=::NewEnumerator0
    Sprite Size Mode=::NewEnumerator3
    Uniform Sprite Size=&SelectFloatFromArray(
     Direct Array Index=&SpawnGroup()
     Float Selection Array=DataInterface() {propertyValues {FloatData [35,35,30,30,35,35]}}
    )
    UsePositionOffset=false
    Write Material Random=true
   ) {module=InitializeParticle001}
   &CameraOffset(Camera Offset Amount=10)
  ]
  particleUpdateScript [
   &ParticleState()
   &SkeletalMeshLocation(
    Bone / Socket Index=&SelectIntFromArray(Direct Array Index=$Particles.SpawnGroup, Int Selection Array=DataInterface() {propertyValues {IntData [1,2,3,4]}})
    Bone Sampling Mode=::NewEnumerator4
    MaskSpawnGroup=false
    Skeletal Mesh=$System.Skeletal Mesh
   ) {module=SkeletalMeshLocation001}
   &DynamicMaterialParameters(
    Index 0 Param 1=2
    Index 0 Param 2=&RandomRangeFloat(Minimum=0, Maximum=1)
    Index 0 Param 3=&RandomRangeFloat(Minimum=0, Maximum=1)
    Index 0 Param 4=0.25
    Module.Param0WriteEnabled=true
    Module.Param1WriteEnabled=true
    Module.Param2WriteEnabled=true
    Module.Param3WriteEnabled=true
   )
   &Color(Color=&MakeLinearColorFromVectorAndFloat(
     "Vector (RGB)"=&Lerp_Vector(
      A=&RandomVector(Vector Scale=&MakeCustomFloatFromBool(Bool=$User.IsFrontend, True Float=5, False Float=20))
      B=&MakeVectorFromLinearColorRGB(Color=&MakeCustomLinearColorFromBool(Bool=$User.IsFrontend, True Color=<rgba 0.338 1.4442 2 1>, False Color=<rgba 1.69 7.2211 10 1>))
      Alpha=&FloatFromCurve(FloatCurve=DataInterface() {propertyValues {Curve {keys [
           {interpMode=RCIM_Cubic, leaveTangentWeight=0.6551, tangentMode=RCTM_User, time=0, value=0}
           {arriveTangentWeight=0.1667, interpMode=RCIM_Cubic, leaveTangentWeight=0.1667, tangentMode=RCTM_User, time=1, value=1}
          ]}}})
     )
     "Float (Alpha)"=&FloatFromCurve(FloatCurve=DataInterface() {propertyValues {Curve {keys [
          {arriveTangent=4.4103, interpMode=RCIM_Cubic, leaveTangent=4.4103, leaveTangentWeight=0.6551, tangentMode=RCTM_User, time=0, value=0}
          {arriveTangentWeight=0.1667, interpMode=RCIM_Cubic, leaveTangentWeight=0.1667, tangentMode=RCTM_User, time=0.5, value=1}
          {arriveTangent=-4.0948, tangentMode=RCTM_User, time=1, value=0}
         ]}}})
    ))
  ]
  rendererClasses [&NiagaraSpriteRendererProperties]
  renderers [
   {rendererClass=&NiagaraSpriteRendererProperties}
   {rendererClass=&NiagaraSpriteRendererProperties, rendererIndex=1}
  ]
 }
 Diamonds {
  emitterUpdateScript [
   &EmitterState()
   &SpawnRate(
    SpawnRate=&RandomRangeFloat(Minimum=0, Maximum=5, Evaluation Type=::NewEnumerator1)
    Use Spawn Probability=false
   )
   &SpawnRate(
    SpawnGroup=1
    SpawnRate=&RandomRangeFloat(Minimum=10, Maximum=15, Evaluation Type=::NewEnumerator1)
    Use Spawn Probability=false
   ) {module=SpawnRate001}
  ]
  particleSpawnScript [
   /SetVariables(
    Particles.ShardIndex:NiagaraInt32=&RandomRangeInt(Minimum=0, Maximum=1)
    Particles.VisibilityTag=&SpawnGroup()
   ) {isSetParameters=true}
   &PartitionParticles(Number Of Partitions=4, Particles Per Partition=2, Partition Mode:ENiagara_PartitionMode=::NewEnumerator1)
   &InitializeParticle(
    Color=<rgba 0.3052 0.5715 1 1>
    Lifetime=&MaskFloatBySpawnGroup(
     Masked Float=&RandomRangeFloat(Minimum=1, Maximum=2)
     Passthrough Float=&RandomRangeFloat(Minimum=0.5, Maximum=1)
    )
    Mesh Scale Mode=::NewEnumerator1
    Mesh Uniform Scale Max=0.004
    Mesh Uniform Scale Min=0.002
    Position Mode=::NewEnumerator0
    Sprite Rotation Mode=::NewEnumerator2
    Sprite Size Mode=::NewEnumerator3
    Uniform Sprite Size=1
    Write Material Random=true
   )
   &CameraOffset(Camera Offset Amount=-30) {enabled=false}
   /SetVariables(
    Particles.FlareCameraOffset:NiagaraFloat=&Add_Float(A=$Particles.CameraOffset, B=-5)
    Particles.SoftFlareColor:LinearColor=&RandomRangeLinearColor(Minimum=<rgba 0.071 0.2878 1 0.2>, Maximum=<rgba 0.183 0.9656 1 0.2>)
   ) {isSetParameters=true}
   &SkeletalMeshLocation(
    Bone / Socket Index=&Add_Integer(A=$Particles.PartitionParticles.Partition, B=1)
    Bone Sampling Mode=::NewEnumerator4
    MaskSpawnGroup=false
    Orientation Sampling=::NewEnumerator1
    Position Sampling=::NewEnumerator2
    Sampled Position Offset=&Multiply_VectorByFloat(
     Vector=&SelectVectorFromArray(
      Direct Array Index=&Add_Integer(A=$Particles.PartitionParticles.Partition, B=1)
      Vector Selection Array=$System.DiamondOffset
     )
     Float=&MakeCustomFloatFromBool(Bool=$User.IsFrontend, True Float=0.5, False Float=1)
    )
    Skeletal Mesh=$System.Skeletal Mesh
   )
   &SkeletalMeshLocation(
    Bone / Socket Index=&SelectIntFromArray(
     Direct Array Index=&Add_Integer(A=$Particles.PartitionParticles.Partition, B=1)
     Int Selection Array=$System.CompanionBone
    )
    Bone Sampling Mode=::NewEnumerator4
    MaskSpawnGroup=false
    Orientation Sampling=::NewEnumerator1
    Position Sampling=::NewEnumerator2
    Sampled Position Offset=&Multiply_VectorByFloat(
     Vector=&SelectVectorFromArray(
      Direct Array Index=&SelectIntFromArray(
       Direct Array Index=&Add_Integer(A=$Particles.PartitionParticles.Partition, B=1)
       Int Selection Array=$System.CompanionBone
      )
      Vector Selection Array=$System.DiamondOffset
     )
     Float=&MakeCustomFloatFromBool(Bool=$User.IsFrontend, True Float=0.75, False Float=1)
    )
    Skeletal Mesh=$System.Skeletal Mesh
   ) {module=SkeletalMeshLocation001}
   /SetVariables(Particles.Position=&LerpPosition(
     A=$Particles.SkeletalMeshLocation.SampledPosition
     B=$Particles.SkeletalMeshLocation001.SampledPosition
     Alpha=&RandomRangeFloat(Minimum=0, Maximum=1)
    )) {isSetParameters=true}
   &AddVelocity(
    Rotation Coordinate Space=::World
    Velocity=&RandomRangeVector(Minimum=<0 0 25>, Maximum=<0 0 75>)
    Velocity Speed Scale=&Multiply_Float(
     A=&MakeCustomFloatFromBool(Bool=$User.IsLargeBody, True Float=1.25, False Float=1)
     B=&MaskFloatBySpawnGroup(Masked Float=1, Passthrough Float=0.25)
    )
   )
  ]
  particleUpdateScript [
   &ParticleState()
   &ScaleSpriteSize(
    Uniform Curve Scale=&MakeCustomFloatFromBool(Bool=$User.IsFrontend, True Float=1, False Float=1.5)
    Uniform Curve Sprite Scale=DataInterface() {propertyValues {Curve {keys [
        {arriveTangent=34.3811, interpMode=RCIM_Cubic, leaveTangent=34.3811, tangentMode=RCTM_User, time=0.88, value=0}
        <at 0.95 1 RCIM_Cubic>
        <at 1 0 RCIM_Cubic>
       ]}}}
   )
   &ScaleMeshSize(Scale Factor=&VectorFromFloat(Value=&FloatFromCurve(
      FloatCurve=DataInterface() {propertyValues {Curve {keys [
          {arriveTangent=3.487, interpMode=RCIM_Cubic, leaveTangent=3.487, tangentMode=RCTM_User, time=0, value=0}
          {
           arriveTangent=1.1641
           arriveTangentWeight=0.1389
           interpMode=RCIM_Cubic
           leaveTangent=1.1641
           leaveTangentWeight=0.2778
           tangentMode=RCTM_User
           time=0.5
           value=1
          }
          {arriveTangentWeight=0.2778, interpMode=RCIM_Cubic, leaveTangentWeight=0.1389, tangentMode=RCTM_User, time=0.7, value=1}
          {arriveTangent=-20.8503, interpMode=RCIM_Cubic, leaveTangent=-20.8503, tangentMode=RCTM_User, time=0.95, value=0}
         ]}}}
      Scale Curve=&MakeCustomFloatFromBool(Bool=$User.IsFrontend, True Float=1, False Float=1.5)
     )))
   &Drag(Drag=1.5, Use Linear Drag=true, Use Rotational Drag=true)
   &CurlNoiseForce(
    Bias Noise=false
    Noise Frequency=100
    Noise Strength=&Lerp_Float(
     A=25
     B=100
     Alpha=&FloatFromCurve(
      FloatCurve=DataInterface() {propertyValues {Curve {keys [{arriveTangent=-2.6129, interpMode=RCIM_Cubic, leaveTangent=-2.6129, tangentMode=RCTM_User, time=0, value=1}, <at 1 0>]}}}
      Scale Curve=&MaskFloatBySpawnGroup(Masked Float=0.5, Passthrough Float=2)
     )
    )
    Pan Noise=true
    Pan Noise Field=<0 0 1>
   )
   &SolveForcesAndVelocity(Clamp Velocity=false, Limit Acceleration=false)
   &UpdateMeshOrientation(
    Rotation Rate=&Lerp_Float(A=2, B=0, Alpha=$Particles.NormalizedAge)
    Rotation Vector=&RandomVector(Vector Scale=1)
   )
   /SetVariables(
    Particles.SoftFlareScale:Vector2f=&Multiply_Vector2D_ByFloat(Float=&RandomRangeFloat(Minimum=20, Maximum=35), Vector2D=$Particles.SpriteSize)
    Particles.StarScale:Vector2f=&Multiply_Vector2D_ByFloat(Float=&RandomRangeFloat(Minimum=25, Maximum=40), Vector2D=$Particles.SpriteSize)
   ) {isSetParameters=true}
   /SetVariables(Particles.DebugCameraOffset=30, Particles.DebugSpriteScale:Vector2f=&Vector2DFromFloat(Value=10)) {enabled=false, isSetParameters=true}
   /SetVariables(Particles.ShardScale:Vector3f=&Multiply_VectorByFloat(
     Vector=&RandomRangeVector(Minimum=<0.8 0.8 0.5>, Maximum=<1.5 1.5 1>)
     Float=&FloatFromCurve(
      FloatCurve=DataInterface() {propertyValues {Curve {keys [
          {arriveTangent=2.5517, interpMode=RCIM_Cubic, leaveTangent=2.5517, tangentMode=RCTM_User, time=0, value=0}
          <at 0.6 1 RCIM_Cubic>
          {arriveTangent=-0.9496, tangentMode=RCTM_User, time=1, value=0}
         ]}}}
      Scale Curve=&RandomRangeFloat(Minimum=0.08, Maximum=0.2)
     )
    )) {isSetParameters=true}
  ]
  rendererClasses [&NiagaraMeshRendererProperties, &NiagaraSpriteRendererProperties]
  renderers [
   {rendererClass=&NiagaraMeshRendererProperties, rendererIndex=0}
   {rendererClass=&NiagaraSpriteRendererProperties, rendererIndex=1}
   {rendererClass=&NiagaraSpriteRendererProperties, rendererIndex=2}
   {rendererClass=&NiagaraSpriteRendererProperties, rendererIndex=3}
   {rendererClass=&NiagaraMeshRendererProperties, rendererIndex=4}
  ]
 }
 Flames {
  emitterUpdateScript [
   &EmitterState()
   &SpawnRate(SpawnRate=$System.RibbonSpawnRate, Use Spawn Probability=false) {enabled=false}
   &SpawnRate(SpawnGroup=1, SpawnRate=$System.RibbonSpawnRate, Use Spawn Probability=false) {module=SpawnRate001}
   &SpawnRate(SpawnGroup=2, SpawnRate=$System.RibbonSpawnRate, Use Spawn Probability=false) {module=SpawnRate002}
   &SpawnRate(SpawnGroup=3, SpawnRate=$System.RibbonSpawnRate, Use Spawn Probability=false) {module=SpawnRate003}
   &SpawnRate(SpawnGroup=4, SpawnRate=$System.RibbonSpawnRate, Use Spawn Probability=false) {module=SpawnRate004}
   &SpawnRate(SpawnGroup=5, SpawnRate=$System.RibbonSpawnRate, Use Spawn Probability=false) {module=SpawnRate005}
   &SpawnRate(SpawnGroup=6, SpawnRate=$System.RibbonSpawnRate, Use Spawn Probability=false) {module=SpawnRate006}
   &self:SetHasSpawned() {module=SetHasSpawned}
  ]
  particleSpawnScript [
   &InitializeParticle(
    Color=&MakeCustomLinearColorFromBool(Bool=$User.IsFrontend, True Color=<rgba 0.7794 0.8779 0.8854 0.25>, False Color=<rgba 0.7794 0.8779 0.8854 0.6>)
    Lifetime=&SelectFloatFromArray(
     Direct Array Index=&SpawnGroup()
     Float Selection Array=DataInterface() {propertyValues {FloatData [0.5,0.9,0.9,0.75,0.75,0.5,0.5]}}
    )
    Ribbon Width=&SelectFloatFromArray(
     Direct Array Index=&SpawnGroup()
     Float Selection Array=DataInterface() {propertyValues {FloatData [50,35,35,40,40,60,60]}}
    )
    Ribbon Width Mode=::NewEnumerator1
    UsePositionOffset=false
    Write Material Random=true
   )
   /SetVariables(Particles.RibbonID=/Niagara/DynamicInputs/NiagaraID/MakeNiagaraIDFromInt(Index:NiagaraInt32=&SpawnGroup(), Acquire Tag:NiagaraInt32=0)) {isSetParameters=true}
   &AddVelocity(Rotation Coordinate Space=::World, Velocity=<0 0 150>)
   &SkeletalMeshLocation(
    Bone / Socket Index=&SpawnGroup()
    Bone Sampling Mode=::NewEnumerator4
    MaskSpawnGroup=false
    Orientation Sampling=::NewEnumerator1
    Sampled Position Offset=&SelectVectorFromArray(Direct Array Index=&SpawnGroup(), Vector Selection Array=$System.BoneOffset)
    Skeletal Mesh=$System.Skeletal Mesh
   ) {module=SkeletalMeshLocation001}
   &SkeletalMeshLocation(
    Bone / Socket Index=&SelectIntFromArray(Direct Array Index=&SpawnGroup(), Int Selection Array=$System.CompanionBone)
    Bone Sampling Mode=::NewEnumerator4
    MaskSpawnGroup=false
    Orientation Sampling=::NewEnumerator1
    Position Sampling=::NewEnumerator2
    Sampled Position Offset=&SelectVectorFromArray(Direct Array Index=&SpawnGroup(), Vector Selection Array=$System.BoneOffset)
    Skeletal Mesh=$System.Skeletal Mesh
   ) {module=SkeletalMeshLocation003}
   &InheritVelocity(
    Inherited Velocity Amount Scale=<-0.5 -0.5 0.1>
    Speed Limit=true
    UseSpeedThreshold=false
    Velocity Source=&Lerp_Vector(A=$Particles.SkeletalMeshLocation001.SampledVelocity, B=$Particles.SkeletalMeshLocation003.SampledVelocity, Alpha=0.5)
   )
   /SetVariables(Particles.RibbonFacing=&MakeCustomVectorFromBool(
     Bool=&SetBoolByIntComparison(A=&SpawnGroup(), B=0, Comparison Type=::NewEnumerator4)
     True Vector=<1 0 0>
     False Vector=&Subtract_Vector(A=$Particles.SkeletalMeshLocation001.SampledPosition, B=$Particles.SkeletalMeshLocation003.SampledPosition)
    )) {isSetParameters=true}
   &DynamicMaterialParameters(
    Index 0 Param 1=&RandomRangeFloat(Minimum=0, Maximum=0.5)
    Module.Index0Param0Write=true
    Module.Index0Param1Write=true
    Module.Index0Param2Write=false
    Module.Index0Param3Write=false
   )
   &CameraOffset()
  ]
  particleUpdateScript [
   &ParticleState()
   &SkeletalMeshLocation(
    Bone / Socket Index=&SpawnGroup()
    Bone Sampling Mode=::NewEnumerator4
    MaskSpawnGroup=false
    Orientation Sampling=::NewEnumerator1
    Position Sampling=::NewEnumerator2
    Sampled Position Offset=&SelectVectorFromArray(Direct Array Index=&SpawnGroup(), Vector Selection Array=$System.BoneOffset)
    Skeletal Mesh=$System.Skeletal Mesh
   )
   &SkeletalMeshLocation(
    Bone / Socket Index=&SelectIntFromArray(Direct Array Index=&SpawnGroup(), Int Selection Array=$System.CompanionBone)
    Bone Sampling Mode=::NewEnumerator4
    MaskSpawnGroup=false
    Orientation Sampling=::NewEnumerator1
    Position Sampling=::NewEnumerator2
    Sampled Position Offset=&SelectVectorFromArray(Direct Array Index=&SpawnGroup(), Vector Selection Array=$System.BoneOffset)
    Skeletal Mesh=$System.Skeletal Mesh
   ) {module=SkeletalMeshLocation002}
   &Drag(Use Linear Drag=true, Use Rotational Drag=true)
   &CurlNoiseForce(Bias Noise=false, Noise Strength=200, Pan Noise=true, Pan Noise Field=<0 0 1>)
   &SolveForcesAndVelocity(Clamp Velocity=false, Limit Acceleration=false)
   &self:IsFirstParticle(
    Locked Position:NiagaraPosition=&Lerp_Vector(
     A=Linked() {linkedVariable {name=Particles.SkeletalMeshLocation.SampledPosition, type=NiagaraPosition}}
     B=$Particles.SkeletalMeshLocation002.SampledPosition
     Alpha=0.5
    )
    Set Position=true
    Spawn Group:NiagaraInt32=&SpawnGroup()
   ) {module=IsFirstParticle}
   &self:HighSpeedControl(
    Max Expected Owner Speed=500
    Max Visible Distance=100
    Negligible Owner Speed=50
    Source Position:NiagaraPosition=$Output.IsFirstParticle.LockedPosition
   ) {module=HighSpeedControl}
   /SetVariables(Particles.Color=&MultiplyLinearColor(
     Linear Color=&MultiplyLinearColor(
      Linear Color=$Particles.Initial.Color
      Scale Factor=&MakeVector4_FromLinearColor(LinearColor=&ColorFromCurve(ColorCurve=DataInterface() {propertyValues {
          AlphaCurve {keys [<at 0 1>,<at 0.4763 0.4>,<at 0.7785 0>]}
          BlueCurve {keys [<at 0 8>,<at 0.4086 1>,<at 1 1>]}
          GreenCurve {keys [<at 0 3.7102>,<at 0.4086 0.8468>,<at 1 1>]}
          RedCurve {keys [<at 0 4.1337>,<at 0.4086 0.5938>,<at 1 1>]}
         }}))
     )
     Scale Factor=&MakeFloat4FromVectorAndFloat(
      "Vector (RGB)"=<1 1 1>
      "Float (Alpha)"=&FloatFromCurve(
       FloatCurve=DataInterface() {propertyValues {Curve {keys [
           {arriveTangent=-0.7548, interpMode=RCIM_Cubic, leaveTangent=-0.7548, tangentMode=RCTM_User, time=0, value=1}
           {arriveTangent=-0.5922, tangentMode=RCTM_User, time=1, value=0}
          ]}}}
       CurveIndex=$Particles.SpeedFade
      )
     )
    )) {isSetParameters=true}
   &ScaleRibbonWidth(Ribbon Width Scale=&FloatFromCurve(
     FloatCurve=DataInterface() {propertyValues {Curve {keys [
         {interpMode=RCIM_Cubic, leaveTangentWeight=0.379, tangentMode=RCTM_User, time=0, value=1}
         {arriveTangent=-1.8425, tangentMode=RCTM_User, time=1, value=0}
        ]}}}
     Scale Curve=&MakeCustomFloatFromBool(Bool=$User.IsLargeBody, True Float=1.1, False Float=1)
    ))
   &Color(Color=&ColorFromCurve(
     ColorCurve=DataInterface() {propertyValues {
       AlphaCurve {keys [<at 0 1 RCIM_Cubic>]}
       BlueCurve {keys [<at 0 0 RCIM_Cubic>,<at 0.5 0>,<at 1 1 RCIM_Cubic>]}
       GreenCurve {keys [<at 0 0 RCIM_Cubic>,<at 0.5 1>,<at 1 0 RCIM_Cubic>]}
       RedCurve {keys [<at 0 1 RCIM_Cubic>,<at 0.5 0>,<at 1 0 RCIM_Cubic>]}
      }}
     CurveIndex=$Particles.SpeedFade
    )) {enabled=false}
  ]
  rendererClasses [&NiagaraRibbonRendererProperties]
  renderers [{rendererClass=&NiagaraRibbonRendererProperties}]
 }
 Glows {
  emitterUpdateScript [
   &EmitterState()
   &SpawnBurst_Instantaneous(Spawn Count=2, Use Spawn Probability=false, UseLoopCountLimit=true)
   &SpawnBurst_Instantaneous(Spawn Count=6, Spawn Group=1, Use Spawn Probability=false, UseLoopCountLimit=true) {module=SpawnBurst_Instantaneous001}
  ]
  particleSpawnScript [
   &InitializeParticle(
    Color Mode=::NewEnumerator0
    Lifetime=2
    Position Offset=<0 -30 110>
    Sprite Rotation Mode=::NewEnumerator2
    Sprite Size Mode=::NewEnumerator3
    Uniform Sprite Size=&MaskFloatBySpawnGroup(
     Masked Float=130
     Passthrough Float=&MakeCustomFloatFromBool(Bool=$User.IsLargeBody, True Float=80, False Float=65)
    )
    UsePositionOffset=false
    Write Material Random=true
   ) {module=InitializeParticle001}
   /SetVariables(
    Particles.PickBone:NiagaraInt32=&MaskIntBySpawnGroup(
     Masked Int=0
     Passthrough Int=&Add_Integer(A=&ReturnExecIndex(), B=1)
     Spawn Group Mask=0
    )
    Particles.VisibilityTag=&SpawnGroup()
   ) {isSetParameters=true}
  ]
  particleUpdateScript [
   &ParticleState(Kill Particles When Lifetime Has Elapsed=false, Let Infinitely Lived Particles Die When Emitter Deactivates=true)
   &SpriteRotationRate(Rotation Rate=&SelectFloatFromArray(
     Direct Array Index=&ReturnExecIndex()
     Float Selection Array=DataInterface() {propertyValues {FloatData [-15,20]}}
    ))
   /SetVariables(Transient.CurveOffset:NiagaraFloat=&ModuloFloat(
     A=&Add_Float(
      A=$System.NormalizedLoopAge
      B=&Multiply_FloatByInt(
       Float=&MaskFloatBySpawnGroup(Masked Float=0.5, Passthrough Float=0.1667)
       Integer=&ReturnExecIndex()
      )
     )
     B=1
    )) {isSetParameters=true}
   /SetVariables(
    Particles.BodyColor:LinearColor=&MultiplyLinearColor(
     Linear Color=&MakeCustomLinearColorFromBool(Bool=$User.IsFrontend, True Color=<rgba 2 2 2 0.1>, False Color=<rgba 1 1 1 0.5>)
     Scale Factor=&MakeVector4_FromLinearColor(LinearColor=&ColorFromCurve(
       ColorCurve=DataInterface() {propertyValues {
         AlphaCurve {keys [<at 0 1>,<at 0.4671 0.6>,<at 1 1>]}
         BlueCurve {keys [<at 0.1217 0.8714>,<at 0.3306 1>,<at 0.4729 0.7292>,<at 0.6336 0.8714>,<at 0.8205 0.8714>]}
         GreenCurve {keys [<at 0.1217 0.3515>,<at 0.3306 0.3241>,<at 0.4729 0.2521>,<at 0.6336 0.1894>,<at 0.8205 0.3515>]}
         RedCurve {keys [<at 0.1217 0.227>,<at 0.3306 0.3406>,<at 0.4729 0.1822>,<at 0.6336 0.2246>,<at 0.8205 0.227>]}
        }}
       CurveIndex=$Transient.CurveOffset
      ))
    )
    Particles.HeadColor:LinearColor=&MultiplyLinearColor(
     Linear Color=&HueShiftLinearColor(
      Hue=&RandomRangeFloat(Minimum=-0.1, Maximum=0.1)
      Saturation=1
      Value=1
      Linear Color=&MakeCustomLinearColorFromBool(Bool=$User.IsFrontend, True Color=<rgba 0.598 0.6062 1 0.25>, False Color=<rgba 0.598 0.6062 1 0.75>)
     )
     Scale Factor=&MakeVector4_FromLinearColor(LinearColor=&ColorFromCurve(
       ColorCurve=DataInterface() {propertyValues {
         AlphaCurve {keys [<at 0 1>,<at 0.4671 0.6>,<at 1 1>]}
         BlueCurve {keys [<at 0.1217 0.7605>,<at 0.3306 1>,<at 0.4729 0.7292>,<at 0.6336 0.8714>,<at 0.8205 0.7605>]}
         GreenCurve {keys [<at 0.1217 0.6308>,<at 0.3306 0.9701>,<at 0.4729 0.4584>,<at 0.6336 0.5208>,<at 0.8205 0.6308>]}
         RedCurve {keys [<at 0.1217 0.4287>,<at 0.3306 0.7067>,<at 0.4729 0.3955>,<at 0.6336 0.5719>,<at 0.8205 0.4287>]}
        }}
       CurveIndex=$Transient.CurveOffset
      ))
    )
   ) {isSetParameters=true}
   &DynamicMaterialParameters(
    Index 0 Param 1=1
    Index 0 Param 2=&RandomRangeFloat(Minimum=0, Maximum=0.2)
    Module.Index0Param0Write=true
    Module.Index0Param1Write=true
    Module.Index0Param2Write=false
    Module.Index0Param3Write=false
   )
   &SkeletalMeshLocation(
    Bone / Socket Index=$Particles.PickBone
    Bone Sampling Mode=::NewEnumerator4
    MaskSpawnGroup=false
    Orientation Sampling=::NewEnumerator1
    Sampled Position Offset=&MaskVectorBySpawnGroup(Masked Vector=<-15 0 0>, Passthrough Vector=<0 0 0>)
    Skeletal Mesh=$System.Skeletal Mesh
   )
   &CameraOffset(Camera Offset Amount=&SelectFloatFromArray(Direct Array Index=&SpawnGroup(), Float Selection Array=DataInterface() {propertyValues {FloatData [-40,-20]}}))
  ]
  rendererClasses [&NiagaraSpriteRendererProperties]
  renderers [
   {rendererClass=&NiagaraSpriteRendererProperties}
   {rendererClass=&NiagaraSpriteRendererProperties, rendererIndex=1}
  ]
 }
 Minimal {
  bEnabled=false
  emitterUpdateScript [
   &EmitterState()
   &SpawnBurst_Instantaneous(Spawn Count=9, Use Spawn Probability=false, UseLoopCountLimit=true)
  ]
  particleSpawnScript [
   &InitializeParticle(Color Mode=::NewEnumerator0, Lifetime=5, UsePositionOffset=false, Write Material Random=true)
   &PartitionParticles(Number Of Partitions=9, Write Partition to Ribbon ID=::NewEnumerator0)
  ]
  particleUpdateScript [
   &ParticleState(Kill Particles When Lifetime Has Elapsed=false, Let Infinitely Lived Particles Die When Emitter Deactivates=true)
   &SkeletalMeshLocation(
    Bone / Socket Index=$Particles.PartitionParticles.Partition
    Bone Sampling Mode=::NewEnumerator4
    MaskSpawnGroup=false
    Sampled Position Offset=&Multiply_VectorByFloat(
     Vector=&SelectVectorFromArray(Direct Array Index=$Particles.PartitionParticles.Partition, Vector Selection Array=$System.DiamondOffset)
     Float=0.5
    )
    Skeletal Mesh=$System.Skeletal Mesh
   )
  ]
  rendererClasses [&NiagaraMeshRendererProperties]
  renderers [{rendererClass=&NiagaraMeshRendererProperties, rendererIndex=0}]
 }
 Minimal001 {
  bEnabled=false
  emitterUpdateScript [
   &EmitterState()
   &SpawnBurst_Instantaneous(Use Spawn Probability=false, UseLoopCountLimit=false)
  ]
  particleSpawnScript [&InitializeParticle(Color Mode=::NewEnumerator0, Lifetime=5, Position Offset=<0 0 200>, UsePositionOffset=true, Write Material Random=true)]
  particleUpdateScript [
   &ParticleState(Kill Particles When Lifetime Has Elapsed=false)
   &DynamicMaterialParameters(
    Index 0 Param 1=&VectorLength(Vector=$Engine.Owner.Velocity)
    Module.Index0Param0Write=true
    Module.Index0Param1Write=true
    Module.Index0Param2Write=true
    Module.Index0Param3Write=true
   )
  ]
  rendererClasses [&NiagaraSpriteRendererProperties]
  renderers [{rendererClass=&NiagaraSpriteRendererProperties}]
 }
 Smoke {
  emitterUpdateScript [
   &EmitterState()
   &SpawnRate(SpawnRate=20, Use Spawn Probability=false)
  ]
  particleSpawnScript [
   /SetVariables(Particles.PickBone:NiagaraInt32=&SelectIntFromArray(
     Array Sampling Mode=::NewEnumerator1
     Direct Array Index=?
     Int Selection Array=DataInterface() {propertyValues {IntData [0,1,2,3,3,4,4,5,5,5,6,6,6]}}
     Random Seed=0
    )) {isSetParameters=true}
   &SkeletalMeshLocation(
    Bone / Socket Index=$Particles.PickBone
    Bone Sampling Mode=::NewEnumerator4
    MaskSpawnGroup=false
    Orientation Sampling=::NewEnumerator1
    Position Sampling=::NewEnumerator2
    Skeletal Mesh=$System.Skeletal Mesh
   )
   &SkeletalMeshLocation(
    Bone / Socket Index=&SelectIntFromArray(Direct Array Index=$Particles.PickBone, Int Selection Array=$System.CompanionBone)
    Bone Sampling Mode=::NewEnumerator4
    MaskSpawnGroup=false
    Orientation Sampling=::NewEnumerator1
    Position Sampling=::NewEnumerator2
    Skeletal Mesh=$System.Skeletal Mesh
   ) {module=SkeletalMeshLocation001}
   &InitializeParticle(
    Color=&RandomRangeLinearColor(
     Minimum=&RandomRangeLinearColor(Minimum=<rgba 0.6604 0 1 0.7>, Maximum=<rgba 0.3798 0 1 0.7>)
     Maximum=&RandomRangeLinearColor(Minimum=<rgba 0 1 0.3879 0.7>, Maximum=<rgba 0 0.4678 1 0.7>)
    )
    Lifetime=&UniformRangedFloat(Minimum=0.75, Maximum=1)
    Position=&LerpPosition(
     A=$Particles.SkeletalMeshLocation.SampledPosition
     B=$Particles.SkeletalMeshLocation001.SampledPosition
     Alpha=&RandomRangeFloat(Minimum=0, Maximum=1)
    )
    Position Mode=::NewEnumerator1
    Sprite Rotation Mode=::NewEnumerator2
    Sprite Size Mode=::NewEnumerator1
    Uniform Sprite Size Max=60
    Uniform Sprite Size Min=40
    UsePositionOffset=false
    Write Material Random=true
   )
   &DynamicMaterialParameters(
    Index 0 Param 1=&RandomRangeFloat(Minimum=-0.05, Maximum=0.05)
    Index 0 Param 2=&RandomRangeFloat(Minimum=0, Maximum=32)
    Module.Index0Param0Write=true
    Module.Index0Param1Write=true
    Module.Index0Param2Write=false
    Module.Index0Param3Write=false
   )
   &CameraOffset(Camera Offset Amount=-10)
  ]
  particleUpdateScript [
   &ParticleState()
   &ScaleColor(
    Scale Alpha=&FloatFromCurve(
     FloatCurve=DataInterface() {propertyValues {Curve {keys [
         {arriveTangent=4.2554, interpMode=RCIM_Cubic, leaveTangent=4.2554, tangentMode=RCTM_User, time=0, value=0}
         {arriveTangent=0.002, interpMode=RCIM_Cubic, leaveTangent=0.002, time=0.5, value=1}
         {arriveTangent=-3.8631, interpMode=RCIM_Cubic, leaveTangent=-3.8631, tangentMode=RCTM_User, time=1.0011, value=0.002}
        ]}}}
     Scale Curve=&MakeCustomFloatFromBool(Bool=$User.IsFrontend, True Float=0.15, False Float=0.3)
    )
    Scale RGB=&VectorFromCurve(
     VectorCurve=DataInterface() {propertyValues {XCurve {keys [<at 0 1>,<at 1 0>]}, YCurve {keys [<at 0 1>,<at 1 0>]}, ZCurve {keys [<at 0 1>,<at 1 2>]}}}
     Scale Curve=<10 10 10>
    )
    ScaleA=true
    ScaleRGB=true
   )
   &ScaleSpriteSize(
    Uniform Curve Scale=&SelectFloatFromArray(Direct Array Index=$Particles.PickBone, Float Selection Array=DataInterface() {propertyValues {FloatData [1.1,1.1,1.1,0.75]}})
    Uniform Curve Sprite Scale=DataInterface() {propertyValues {Curve {keys [
        {arriveTangent=1.5814, interpMode=RCIM_Cubic, leaveTangent=1.5814, tangentMode=RCTM_User, time=0, value=0.2}
        {interpMode=RCIM_Cubic, leaveTangent=-0.4663, time=1, value=1}
       ]}}}
   )
  ]
  rendererClasses [&NiagaraSpriteRendererProperties]
  renderers [{rendererClass=&NiagaraSpriteRendererProperties}]
 }
 Stars {
  emitterUpdateScript [
   &EmitterState()
   &SpawnRate(SpawnRate=10, Use Spawn Probability=false)
  ]
  particleSpawnScript [
   &PartitionParticles(Number Of Partitions=5)
   &InitializeParticle(
    Color Mode=::NewEnumerator0
    Lifetime=&RandomRangeFloat(Minimum=1.2, Maximum=1)
    Sprite Rotation Mode=::NewEnumerator2
    Sprite Size Mode=::NewEnumerator3
    Uniform Sprite Size=&MakeCustomFloatFromBool(
     Bool=$User.IsLargeBody
     True Float=&RandomRangeFloat(Minimum=20, Maximum=45)
     False Float=&RandomRangeFloat(Minimum=15, Maximum=35)
    )
    UsePositionOffset=false
    Write Material Random=true
   )
   &AddVelocity(
    Velocity=&RandomRangeVector(Minimum=<0 -20 25>, Maximum=<0 20 75>)
    Velocity Speed Scale=&MakeCustomFloatFromBool(Bool=$User.IsLargeBody, True Float=1.5, False Float=1.4)
   )
   &CameraOffset(Camera Offset Amount=-30)
  ]
  particleUpdateScript [
   &ParticleState()
   &SpriteRotationRate(Rotation Rate=&FloatFromCurve(
     FloatCurve=DataInterface() {propertyValues {Curve {keys [{arriveTangent=-2.7442, interpMode=RCIM_Cubic, leaveTangent=-2.7442, tangentMode=RCTM_User, time=0, value=1}, <at 1 0>]}}}
     Scale Curve=&RandomRangeFloat(Minimum=-720, Maximum=720)
    ))
   &ScaleSpriteSize(
    Uniform Curve Scale=&Waveform(
     "Waveform [1]"=::NewEnumerator0
     "[1] Phase Scale"=&FloatFromCurve(
      FloatCurve=DataInterface() {propertyValues {Curve {keys [
          {arriveTangent=-6.5624, interpMode=RCIM_Cubic, leaveTangent=-6.5624, tangentMode=RCTM_User, time=0, value=1}
          {arriveTangent=-1.0855, interpMode=RCIM_Cubic, leaveTangent=-1.0855, tangentMode=RCTM_User, time=0.129, value=0.4043}
          <at 1 0>
         ]}}}
      Scale Curve=&RandomRangeFloat(Minimum=5, Maximum=10)
     )
     "[1] Frequency"=1
     "[1] Amplitude Min/Max"=&Vector2DFromCurve(Vector2Curve=DataInterface() {propertyValues {XCurve {keys [<at 0 0.25>,<at 1 0.5>]}, YCurve {keys [<at 1 1>]}}}, ScaleCurve=<1 1>)
    )
    Uniform Curve Sprite Scale=DataInterface() {propertyValues {Curve {keys [
        {tangentMode=RCTM_User, time=0, value=0}
        {arriveTangent=1.4286, interpMode=RCIM_Cubic, leaveTangent=-0.3443, tangentMode=RCTM_Break, time=0.15, value=1.2}
        {arriveTangent=-0.4706, interpMode=RCIM_Cubic, leaveTangent=-0.4706, time=0.7, value=1}
        <at 1 0.8>
       ]}}}
   )
   &SkeletalMeshLocation(
    Bone / Socket Index=&SelectIntFromArray(Direct Array Index=$Particles.PartitionParticles.Partition, Int Selection Array=DataInterface() {propertyValues {IntData [0,1,2,3,4]}})
    Bone Sampling Mode=::NewEnumerator4
    MaskSpawnGroup=false
    Position Sampling=::NewEnumerator1
    Sampled Position Offset=<-10 0 0>
    Skeletal Mesh=$System.Skeletal Mesh
    Source Movement Scale:Vector3f=<0 0 0>
   ) {module=SkeletalMeshLocation002}
   &Drag(
    Drag=&FloatFromCurve(
     FloatCurve=DataInterface() {propertyValues {Curve {keys [{arriveTangent=2.934, interpMode=RCIM_Cubic, leaveTangent=2.934, tangentMode=RCTM_User, time=0, value=0}, <at 1 1 RCIM_Cubic>]}}}
     Scale Curve=5
    )
    Rotational Drag=4
    Use Linear Drag=true
    Use Rotational Drag=false
   )
   &SolveForcesAndVelocity(Clamp Velocity=false, Limit Acceleration=false)
   &Color(Color=&MakeLinearColorFromVectorAndFloat(
     "Vector (RGB)"=&Lerp_Vector(
      A=&RandomVector(Vector Scale=5)
      B=&MakeVectorFromLinearColorRGB(Color=<rgba 0.169 0.7221 1 1>)
      Alpha=&FloatFromCurve(FloatCurve=DataInterface() {propertyValues {Curve {keys [
           {arriveTangent=3.0828, interpMode=RCIM_Cubic, leaveTangent=3.0828, tangentMode=RCTM_User, time=0, value=0}
           {tangentMode=RCTM_User, time=1, value=1}
          ]}}})
     )
     "Float (Alpha)"=&Lerp_Float(
      A=&MakeCustomFloatFromBool(Bool=$User.IsFrontend, True Float=0.5, False Float=1)
      B=0
      Alpha=$Particles.NormalizedAge
     )
    ))
  ]
  rendererClasses [&NiagaraSpriteRendererProperties]
  renderers [{rendererClass=&NiagaraSpriteRendererProperties}]
 }
}
summary {
 systemName=NS_Aura_Diamond
 emitters [
  {emitterName=Flames, rendererClasses [&NiagaraRibbonRendererProperties]}
  {emitterName=Decal, rendererClasses [&NiagaraSpriteRendererProperties]}
  {emitterName=Stars, rendererClasses [&NiagaraSpriteRendererProperties]}
  {emitterName=Diamonds, rendererClasses [&NiagaraMeshRendererProperties, &NiagaraSpriteRendererProperties]}
  {emitterName=Smoke, rendererClasses [&NiagaraSpriteRendererProperties]}
  {bEnabled=false, emitterName=Minimal, rendererClasses [&NiagaraMeshRendererProperties]}
  {bEnabled=false, emitterName=Minimal001, rendererClasses [&NiagaraSpriteRendererProperties]}
  {emitterName=Bits, rendererClasses [&NiagaraSpriteRendererProperties]}
  {emitterName=Glows, rendererClasses [&NiagaraSpriteRendererProperties]}
 ]
 userVariables [
  User.IsFrontend {type=NiagaraBool}
  User.IsLargeBody {type=NiagaraBool}
 ]
}
